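# Editor utility actor that traces between two editable local points and displays the physical material it hits, for testing track surfaces.
# Construction places an arrow and text label, runs a complex FortBuildingMeshPhysics line trace, then on a hit writes the material name and shortens the arrow to the hit.
# It hooks editor selection-change events to refresh the trace, unbinding when destroyed.
# A self-resetting RerunAllTraceActors flag re-runs every such actor in the level.
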EditorUtilityBlueprint /DelMarTools/EditorUtilities/GeometryScripts/TrackCombiner/Actors/EUA_TestTracePhysMat
graphs [
 DoConstructLogic {nodes [
   DoConstructLogic() {is=title K2Node_FunctionEntry}
   > n1 = Sequence!(execute) -> (then_0, then_1, then_2:Exec, then_3:Exec)
   Set Relative Location!(
    execute=@n1.then_1
    self:Scene Component Object Reference=Get Arrow() -> (Arrow:Arrow Component Object Reference)
    NewLocation:Vector=Get TraceStart() -> (TraceStart:Vector)
    bSweep:bool=false
    bTeleport:bool=false
   ) -> (then, SweepHitResult:Hit Result Structure) {is=title K2Node_CallFunction}
   n2 = vector - vector(A=Get TraceEnd() -> (TraceEnd:Vector), B=Get TraceStart())
   > Set Relative Rotation!(
    execute
    self:Scene Component Object Reference=Get Arrow()
    NewRotation:Rotator=MakeRotFromX!(X:"Vector (by ref)"=@n2) -> (ReturnValue:Rotator)
    bSweep:bool=false
    bTeleport:bool=false
   ) -> (then, SweepHitResult:Hit Result Structure) {is=title K2Node_CallFunction}
   > Set ArrowLength!(execute, ArrowLength:float=Vector Length(A=@n2), self:Arrow Component Object Reference=Get Arrow()) -> (then, Output_Get:float)
   Set Relative Location(
    execute=@n1.then_2
    self=Get TextRender() -> (TextRender:Text Render Component Object Reference)
    NewLocation=vector + vector(A=Get TraceStart(), B="0,0,10")
    bSweep=false
    bTeleport=false
   )
   n3 = Line Trace By Profile!(
    execute=@n1.then_3
    WorldContextObject:Object Reference
    Start:Vector=TransformLocation!(T:"Transform (by ref)"=Get Actor Transform(), Location:Vector=Get TraceStart()) -> (ReturnValue:Vector)
    End:Vector=TransformLocation(T=Get Actor Transform(), Location=Get TraceEnd())
    ProfileName:Name=FortBuildingMeshPhysics
    bTraceComplex:bool=true
    ActorsToIgnore:Array of Actor Object References
    DrawDebugType:EDrawDebugTrace Enum
    bIgnoreSelf:bool=true
    TraceColor:Linear Color Structure=(R=1,G=0,B=0,A=1)
    TraceHitColor:Linear Color Structure=(R=0,G=1,B=0,A=1)
    DrawTime:float=5
   ) -> (then, OutHit:Hit Result Structure, ReturnValue:bool)
   > Branch(Condition=@n3.ReturnValue)
   n4 = BreakHitResult!(Hit:"Hit Result Structure (by ref)"=@n3) -> (
    bBlockingHit:bool
    bInitialOverlap:bool
    Time:float
    Distance:float
    Location:Vector
    ImpactPoint:Vector
    Normal:Vector
    ImpactNormal:Vector
    PhysMat:Physical Material Object Reference
    HitActor:Actor Object Reference
    HitComponent:Primitive Component Object Reference
    HitBoneName:Name
    BoneName:Name
    HitItem:int
    ElementIndex:int
    FaceIndex:int
    TraceStart:Vector
    TraceEnd:Vector
   )
   > Set Text!(
    execute
    self:Text Render Component Object Reference=Get TextRender()
    Value:"Text (by ref)"="To Text (String)"(InString="To String (Object)"!(InObj:Object Reference=@n4.PhysMat) -> (ReturnValue:String))
   ) -> (then) {is=title K2Node_CallFunction}
   n5 = Get Actor Transform()
   n6 = Get TraceStart()
   > Set ArrowLength(
    ArrowLength="Distance (Vector)"!(V1:Vector=TransformLocation(T=@n5, Location=@n6), V2:Vector=@n4.Location) -> (ReturnValue:double)
    self=Get Arrow()
   )
  ]}
 DoConsuctOnAll {nodes [
   DoConsuctOnAll() {is=title K2Node_FunctionEntry}
   > n1 = GetAllLevelActors(self=Get EditorActorSubsystem())
   > n2 = Filter by Class!(
    execute
    TargetArray:Array of Object References=@n1
    ObjectClass:Object Class Reference=&/EUA_TestTracePhysMat.EUA_TestTracePhysMat_C
    FilterType:EEditorScriptingFilterType Enum
   ) -> (then, ReturnValue:Array of EUA Test Trace Phys Mat Object References)
   > n3 = For Each Loop(Array:Array of EUA Test Trace Phys Mat Object References=@n2) -> (Array Element:EUA Test Trace Phys Mat Object Reference)
   > DoConstructLogic(self=@n3) {is=title K2Node_CallFunction}
  ]}
 EventGraph {nodes [
   Event Destroyed() {is=title K2Node_Event}
   > n1 = GetSelectionSet(self=Get LevelEditorSubsystem())
   n2 = Create Event()
   > Unbind Event from On Selection Change(self:Typed Element Selection Set Object Reference=@n1, Delegate=@n2) {is=title K2Node_RemoveDelegate}
   > Unbind Event from On Pre Selection Change(self:Typed Element Selection Set Object Reference=@n1, Delegate=@n2) {is=title K2Node_RemoveDelegate}
   Event BeginPlay()
   Event ActorBeginOverlap()
   Event Tick!() -> (OutputDelegate, then, DeltaSeconds:float)
  ]}
 HandlePreSelectionChanged {nodes [
   n1 = HandlePreSelectionChanged!() -> (then, SelectionSet:Typed Element Selection Set Object Reference) {is=title K2Node_FunctionEntry}
   > GetSelectedObjects!(execute, self:Typed Element Selection Set Object Reference=@n1, InRequiredClass:Object Class Reference=&Actor) -> (then, ReturnValue:Array of Object References)
   > DoConstructLogic()
  ]}
 InitListeners {nodes [
   n1 = GetSelectionSet(execute=@n4.then, self=Get LevelEditorSubsystem())
   n2 = Create Event()
   n3 = Bind Event to On Selection Change(self:Typed Element Selection Set Object Reference=@n1, Delegate=@n2) {is=title K2Node_AddDelegate}
   Bind Event to On Pre Selection Change(execute=[@n3.then, @n1.then], self:Typed Element Selection Set Object Reference=@n1, Delegate=@n2) {is=title K2Node_AddDelegate}
   n4 = InitListeners() {is=title K2Node_FunctionEntry}
  ]}
 UserConstructionScript {nodes [
   Construction Script()
   > n1 = Sequence!(execute) -> (then_0, then_1)
   > InitListeners() {is=title K2Node_CallFunction}
   n2 = Branch(execute=@n1.then_1, Condition=Get RerunAllTraceActors() -> (RerunAllTraceActors:bool))
   > Set RerunAllTraceActors(RerunAllTraceActors:bool=false) -> (Output_Get:bool)
   > DoConsuctOnAll() {is=title K2Node_CallFunction}
   DoConstructLogic(execute=@n2.else)
  ]}
]
variables [
 RerunAllTraceActors {type=boolean}
 TraceEnd {type=Vector, default=<0 0 -500>}
 TraceStart {type=Vector, default=<0 0 500>}
]
